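# Physics asset for the DualParadox male medium body's skirt, simulating twelve lower-shirt bone chains around the hips.
# Each chain runs from a zero-size MSTR root through two capsule segments, with joints limited to a wide hinge-like swing on one axis and almost no twist or side swing.
# Roots and both thighs attach to the pelvis with free angular constraints, and pelvis and thigh capsules provide leg and hip volumes.
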PhysicsAsset /BRCosmetics/Characters/Player/Male/Medium/Bodies/M_MED_DualParadox/Meshes/M_MED_DualParadox_SkirtPhysics
bodies {
 dyn_back_lf_01_lowershirt_01 [{shapeType=Capsule, length=8, radius=3, shapeName=dyn_back_lf_01_lowershirt_01_capsule, center=<10.1064 0 0.75>, rotation=<rot 0 0 90>}]
 dyn_back_lf_01_lowershirt_02 [{shapeType=Capsule, length=8, radius=3, shapeName=dyn_back_lf_01_lowershirt_02_capsule, center=<10.1064 0 0.75>, rotation=<rot 0 0 90>}]
 dyn_back_lf_01_lowershirt_MSTR [{shapeType=Capsule, length=0, radius=0, shapeName=dyn_back_lf_01_lowershirt_MSTR_capsule}]
 dyn_back_lf_02_lowershirt_01 [{shapeType=Capsule, length=8, radius=3, shapeName=dyn_back_lf_02_lowershirt_01_capsule, center=<10.1064 0 0.75>, rotation=<rot 0 0 90>}]
 dyn_back_lf_02_lowershirt_02 [{shapeType=Capsule, length=8, radius=3, shapeName=dyn_back_lf_02_lowershirt_02_capsule, center=<10.1064 0 0.75>, rotation=<rot 0 0 90>}]
 dyn_back_lf_02_lowershirt_MSTR [{shapeType=Capsule, length=0, radius=0, shapeName=dyn_back_lf_02_lowershirt_MSTR_capsule}]
 dyn_back_rt_01_lowershirt_01 [{shapeType=Capsule, length=8, radius=3, shapeName=dyn_back_rt_01_lowershirt_01_capsule, center=<10.1064 0 0.75>, rotation=<rot 0 0 90>}]
 dyn_back_rt_01_lowershirt_02 [{shapeType=Capsule, length=8, radius=3, shapeName=dyn_back_rt_01_lowershirt_02_capsule, center=<10.1064 0 0.75>, rotation=<rot 0 0 90>}]
 dyn_back_rt_01_lowershirt_MSTR [{shapeType=Capsule, length=0, radius=0, shapeName=dyn_back_rt_01_lowershirt_MSTR_capsule}]
 dyn_back_rt_02_lowershirt_01 [{shapeType=Capsule, length=8, radius=3, shapeName=dyn_back_rt_02_lowershirt_01_capsule, center=<10.1064 0 0.75>, rotation=<rot 0 0 90>}]
 dyn_back_rt_02_lowershirt_02 [{shapeType=Capsule, length=8, radius=3, shapeName=dyn_back_rt_02_lowershirt_02_capsule, center=<10.1064 0 0.75>, rotation=<rot 0 0 90>}]
 dyn_back_rt_02_lowershirt_MSTR [{shapeType=Capsule, length=0, radius=0, shapeName=dyn_back_rt_02_lowershirt_MSTR_capsule}]
 dyn_front_lf_01_lowershirt_01 [{shapeType=Capsule, length=8, radius=3, shapeName=dyn_front_lf_01_lowershirt_01_capsule, center=<10.1064 0 0.75>, rotation=<rot 0 0 90>}]
 dyn_front_lf_01_lowershirt_02 [{shapeType=Capsule, length=8, radius=3, shapeName=dyn_front_lf_01_lowershirt_02_capsule, center=<10.1064 0 0.75>, rotation=<rot 0 0 90>}]
 dyn_front_lf_01_lowershirt_MSTR [{shapeType=Capsule, length=0, radius=0, shapeName=dyn_front_lf_01_lowershirt_MSTR_capsule}]
 dyn_front_lf_02_lowershirt_01 [{shapeType=Capsule, length=8, radius=3, shapeName=dyn_front_lf_02_lowershirt_01_capsule, center=<10.1064 0 0.75>, rotation=<rot 0 0 90>}]
 dyn_front_lf_02_lowershirt_02 [{shapeType=Capsule, length=8, radius=3, shapeName=dyn_front_lf_02_lowershirt_02_capsule, center=<10.1064 0 0.75>, rotation=<rot 0 0 90>}]
 dyn_front_lf_02_lowershirt_MSTR [{shapeType=Capsule, length=0, radius=0, shapeName=dyn_front_lf_02_lowershirt_MSTR_capsule}]
 dyn_front_rt_01_lowershirt_01 [{shapeType=Capsule, length=8, radius=3, shapeName=dyn_front_rt_01_lowershirt_01_capsule, center=<10.1064 0 0.75>, rotation=<rot 0 0 90>}]
 dyn_front_rt_01_lowershirt_02 [{shapeType=Capsule, length=8, radius=3, shapeName=dyn_front_rt_01_lowershirt_02_capsule, center=<10.1064 0 0.75>, rotation=<rot 0 0 90>}]
 dyn_front_rt_01_lowershirt_MSTR [{shapeType=Capsule, length=0, radius=0, shapeName=dyn_front_rt_01_lowershirt_MSTR_capsule}]
 dyn_front_rt_02_lowershirt_01 [{shapeType=Capsule, length=8, radius=3, shapeName=dyn_front_rt_02_lowershirt_01_capsule, center=<10.1064 0 0.75>, rotation=<rot 0 0 90>}]
 dyn_front_rt_02_lowershirt_02 [{shapeType=Capsule, length=8, radius=3, shapeName=dyn_front_rt_02_lowershirt_02_capsule, center=<10.1064 0 0.75>, rotation=<rot 0 0 90>}]
 dyn_front_rt_02_lowershirt_MSTR [{shapeType=Capsule, length=0, radius=0, shapeName=dyn_front_rt_02_lowershirt_MSTR_capsule}]
 dyn_lf_01_lowershirt_01 [{shapeType=Capsule, length=8, radius=3, shapeName=dyn_lf_01_lowershirt_01_capsule, center=<10.1064 0 0.75>, rotation=<rot 0 0 90>}]
 dyn_lf_01_lowershirt_02 [{shapeType=Capsule, length=8, radius=3, shapeName=dyn_lf_01_lowershirt_02_capsule, center=<10.1064 0 0.75>, rotation=<rot 0 0 90>}]
 dyn_lf_01_lowershirt_MSTR [{shapeType=Capsule, length=0, radius=0, shapeName=dyn_lf_01_lowershirt_MSTR_capsule}]
 dyn_lf_02_lowershirt_01 [{shapeType=Capsule, length=8, radius=3, shapeName=dyn_lf_02_lowershirt_01_capsule, center=<10.1064 0 0.75>, rotation=<rot 0 0 90>}]
 dyn_lf_02_lowershirt_02 [{shapeType=Capsule, length=8, radius=3, shapeName=dyn_lf_02_lowershirt_02_capsule, center=<10.1064 0 0.75>, rotation=<rot 0 0 90>}]
 dyn_lf_02_lowershirt_MSTR [{shapeType=Capsule, length=0, radius=0, shapeName=dyn_lf_02_lowershirt_MSTR_capsule}]
 dyn_rt_01_lowershirt_01 [{shapeType=Capsule, length=8, radius=3, shapeName=dyn_rt_01_lowershirt_01_capsule, center=<10.1064 0 0.75>, rotation=<rot 0 0 90>}]
 dyn_rt_01_lowershirt_02 [{shapeType=Capsule, length=8, radius=3, shapeName=dyn_rt_01_lowershirt_02_capsule, center=<10.1064 0 0.75>, rotation=<rot 0 0 90>}]
 dyn_rt_01_lowershirt_MSTR [{shapeType=Capsule, length=0, radius=0, shapeName=dyn_rt_01_lowershirt_MSTR_capsule}]
 dyn_rt_02_lowershirt_01 [{shapeType=Capsule, length=8, radius=3, shapeName=dyn_rt_02_lowershirt_01_capsule, center=<10.1064 0 0.75>, rotation=<rot 0 0 90>}]
 dyn_rt_02_lowershirt_02 [{shapeType=Capsule, length=8, radius=3, shapeName=dyn_rt_02_lowershirt_02_capsule, center=<10.1064 0 0.75>, rotation=<rot 0 0 90>}]
 dyn_rt_02_lowershirt_MSTR [{shapeType=Capsule, length=0, radius=0, shapeName=dyn_rt_02_lowershirt_MSTR_capsule}]
 pelvis [{shapeType=Capsule, length=2, radius=14, shapeName=pelvis_capsule, center=<0 1.3585 0>, rotation=<rot 90 0 0>}]
 thigh_l [
  {shapeType=Capsule, length=24, radius=11, shapeName=thigh_l_capsule, center=<-16 1.5 0>, rotation=<rot 90 0 0>}
  {shapeType=Capsule, length=3, radius=10, shapeName=thigh_l_capsule, center=<-3 1.5 0>, rotation=<rot 0 0 90>}
 ]
 thigh_r [
  {shapeType=Capsule, length=24, radius=11, shapeName=thigh_r_capsule, center=<16 -1.5 0>, rotation=<rot 90 0 0>}
  {shapeType=Capsule, length=3, radius=10, shapeName=thigh_r_capsule, center=<3 -1.5 0>, rotation=<rot 0 0 90>}
 ]
}
constraints [
 {
  bone1Name=dyn_back_lf_01_lowershirt_01
  bone2Name=dyn_back_lf_01_lowershirt_MSTR
  swing1LimitDegrees=2
  swing2LimitDegrees=90
  twistLimitDegrees=2
 }
 {
  bone1Name=dyn_back_lf_01_lowershirt_02
  bone2Name=dyn_back_lf_01_lowershirt_01
  swing1LimitDegrees=2
  swing2LimitDegrees=90
  twistLimitDegrees=2
 }
 {
  bone1Name=dyn_back_lf_02_lowershirt_01
  bone2Name=dyn_back_lf_02_lowershirt_MSTR
  swing1LimitDegrees=2
  swing2LimitDegrees=90
  twistLimitDegrees=2
 }
 {
  bone1Name=dyn_back_lf_02_lowershirt_02
  bone2Name=dyn_back_lf_02_lowershirt_01
  swing1LimitDegrees=2
  swing2LimitDegrees=90
  twistLimitDegrees=2
 }
 {
  bone1Name=dyn_back_rt_01_lowershirt_01
  bone2Name=dyn_back_rt_01_lowershirt_MSTR
  swing1LimitDegrees=2
  swing2LimitDegrees=90
  twistLimitDegrees=2
 }
 {
  bone1Name=dyn_back_rt_01_lowershirt_02
  bone2Name=dyn_back_rt_01_lowershirt_01
  swing1LimitDegrees=2
  swing2LimitDegrees=90
  twistLimitDegrees=2
 }
 {
  bone1Name=dyn_back_rt_02_lowershirt_01
  bone2Name=dyn_back_rt_02_lowershirt_MSTR
  swing1LimitDegrees=2
  swing2LimitDegrees=90
  twistLimitDegrees=2
 }
 {
  bone1Name=dyn_back_rt_02_lowershirt_02
  bone2Name=dyn_back_rt_02_lowershirt_01
  swing1LimitDegrees=2
  swing2LimitDegrees=90
  twistLimitDegrees=2
 }
 {
  bone1Name=dyn_front_lf_01_lowershirt_01
  bone2Name=dyn_front_lf_01_lowershirt_MSTR
  swing1LimitDegrees=2
  swing2LimitDegrees=90
  twistLimitDegrees=2
 }
 {
  bone1Name=dyn_front_lf_01_lowershirt_02
  bone2Name=dyn_front_lf_01_lowershirt_01
  swing1LimitDegrees=2
  swing2LimitDegrees=90
  twistLimitDegrees=2
 }
 {
  bone1Name=dyn_front_lf_02_lowershirt_01
  bone2Name=dyn_front_lf_02_lowershirt_MSTR
  swing1LimitDegrees=2
  swing2LimitDegrees=90
  twistLimitDegrees=2
 }
 {
  bone1Name=dyn_front_lf_02_lowershirt_02
  bone2Name=dyn_front_lf_02_lowershirt_01
  swing1LimitDegrees=2
  swing2LimitDegrees=90
  twistLimitDegrees=2
 }
 {
  bone1Name=dyn_front_rt_01_lowershirt_01
  bone2Name=dyn_front_rt_01_lowershirt_MSTR
  swing1LimitDegrees=2
  swing2LimitDegrees=90
  twistLimitDegrees=2
 }
 {
  bone1Name=dyn_front_rt_01_lowershirt_02
  bone2Name=dyn_front_rt_01_lowershirt_01
  swing1LimitDegrees=2
  swing2LimitDegrees=90
  twistLimitDegrees=2
 }
 {
  bone1Name=dyn_front_rt_02_lowershirt_01
  bone2Name=dyn_front_rt_02_lowershirt_MSTR
  swing1LimitDegrees=2
  swing2LimitDegrees=90
  twistLimitDegrees=2
 }
 {
  bone1Name=dyn_front_rt_02_lowershirt_02
  bone2Name=dyn_front_rt_02_lowershirt_01
  swing1LimitDegrees=2
  swing2LimitDegrees=90
  twistLimitDegrees=2
 }
 {bone1Name=dyn_lf_01_lowershirt_01, bone2Name=dyn_lf_01_lowershirt_MSTR, swing1LimitDegrees=2, swing2LimitDegrees=90, twistLimitDegrees=2}
 {bone1Name=dyn_lf_01_lowershirt_02, bone2Name=dyn_lf_01_lowershirt_01, swing1LimitDegrees=2, swing2LimitDegrees=90, twistLimitDegrees=2}
 {bone1Name=dyn_lf_02_lowershirt_01, bone2Name=dyn_lf_02_lowershirt_MSTR, swing1LimitDegrees=2, swing2LimitDegrees=90, twistLimitDegrees=2}
 {bone1Name=dyn_lf_02_lowershirt_02, bone2Name=dyn_lf_02_lowershirt_01, swing1LimitDegrees=2, swing2LimitDegrees=90, twistLimitDegrees=2}
 {bone1Name=dyn_rt_01_lowershirt_01, bone2Name=dyn_rt_01_lowershirt_MSTR, swing1LimitDegrees=2, swing2LimitDegrees=90, twistLimitDegrees=2}
 {bone1Name=dyn_rt_01_lowershirt_02, bone2Name=dyn_rt_01_lowershirt_01, swing1LimitDegrees=2, swing2LimitDegrees=90, twistLimitDegrees=2}
 {bone1Name=dyn_rt_02_lowershirt_01, bone2Name=dyn_rt_02_lowershirt_MSTR, swing1LimitDegrees=2, swing2LimitDegrees=90, twistLimitDegrees=2}
 {bone1Name=dyn_rt_02_lowershirt_02, bone2Name=dyn_rt_02_lowershirt_01, swing1LimitDegrees=2, swing2LimitDegrees=90, twistLimitDegrees=2}
 {
  bone1Name=dyn_back_lf_02_lowershirt_MSTR
  bone2Name=pelvis
  swing1LimitDegrees=45
  swing1Motion=Free
  swing2LimitDegrees=45
  swing2Motion=Free
  twistLimitDegrees=45
  twistMotion=Free
 }
 {
  bone1Name=dyn_lf_01_lowershirt_MSTR
  bone2Name=pelvis
  swing1LimitDegrees=45
  swing1Motion=Free
  swing2LimitDegrees=45
  swing2Motion=Free
  twistLimitDegrees=45
  twistMotion=Free
 }
 {
  bone1Name=dyn_front_lf_01_lowershirt_MSTR
  bone2Name=pelvis
  swing1LimitDegrees=45
  swing1Motion=Free
  swing2LimitDegrees=45
  swing2Motion=Free
  twistLimitDegrees=45
  twistMotion=Free
 }
 {
  bone1Name=dyn_front_lf_02_lowershirt_MSTR
  bone2Name=pelvis
  swing1LimitDegrees=45
  swing1Motion=Free
  swing2LimitDegrees=45
  swing2Motion=Free
  twistLimitDegrees=45
  twistMotion=Free
 }
 {
  bone1Name=dyn_lf_02_lowershirt_MSTR
  bone2Name=pelvis
  swing1LimitDegrees=45
  swing1Motion=Free
  swing2LimitDegrees=45
  swing2Motion=Free
  twistLimitDegrees=45
  twistMotion=Free
 }
 {
  bone1Name=dyn_back_lf_01_lowershirt_MSTR
  bone2Name=pelvis
  swing1LimitDegrees=45
  swing1Motion=Free
  swing2LimitDegrees=45
  swing2Motion=Free
  twistLimitDegrees=45
  twistMotion=Free
 }
 {
  bone1Name=dyn_back_rt_02_lowershirt_MSTR
  bone2Name=pelvis
  swing1LimitDegrees=45
  swing1Motion=Free
  swing2LimitDegrees=45
  swing2Motion=Free
  twistLimitDegrees=45
  twistMotion=Free
 }
 {
  bone1Name=dyn_rt_01_lowershirt_MSTR
  bone2Name=pelvis
  swing1LimitDegrees=45
  swing1Motion=Free
  swing2LimitDegrees=45
  swing2Motion=Free
  twistLimitDegrees=45
  twistMotion=Free
 }
 {
  bone1Name=dyn_front_rt_01_lowershirt_MSTR
  bone2Name=pelvis
  swing1LimitDegrees=45
  swing1Motion=Free
  swing2LimitDegrees=45
  swing2Motion=Free
  twistLimitDegrees=45
  twistMotion=Free
 }
 {
  bone1Name=dyn_front_rt_02_lowershirt_MSTR
  bone2Name=pelvis
  swing1LimitDegrees=45
  swing1Motion=Free
  swing2LimitDegrees=45
  swing2Motion=Free
  twistLimitDegrees=45
  twistMotion=Free
 }
 {
  bone1Name=dyn_rt_02_lowershirt_MSTR
  bone2Name=pelvis
  swing1LimitDegrees=45
  swing1Motion=Free
  swing2LimitDegrees=45
  swing2Motion=Free
  twistLimitDegrees=45
  twistMotion=Free
 }
 {
  bone1Name=dyn_back_rt_01_lowershirt_MSTR
  bone2Name=pelvis
  swing1LimitDegrees=45
  swing1Motion=Free
  swing2LimitDegrees=45
  swing2Motion=Free
  twistLimitDegrees=45
  twistMotion=Free
 }
 {
  bone1Name=thigh_l
  bone2Name=pelvis
  swing1LimitDegrees=45
  swing1Motion=Free
  swing2LimitDegrees=45
  swing2Motion=Free
  twistLimitDegrees=45
  twistMotion=Free
 }
 {
  bone1Name=thigh_r
  bone2Name=pelvis
  swing1LimitDegrees=45
  swing1Motion=Free
  swing2LimitDegrees=45
  swing2Motion=Free
  twistLimitDegrees=45
  twistMotion=Free
 }
]
