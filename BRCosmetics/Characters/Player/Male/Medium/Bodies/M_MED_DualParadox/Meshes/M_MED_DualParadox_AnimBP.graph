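# Post-process-style AnimBlueprint for the DualParadox male body cosmetic, adding skirt motion and deformation fixes to the copied parent pose.
# A deform rig, thigh/pelvis constraints and a skirt control rig feed RigidBody skirt physics, knee springs and state-driven additive bone offsets.
# The EventGraph derives blend alphas from movement states, damping skirt physics during actions, and overrides gravity.
# It also drives material WPO globals, foot/calf regions and skydive wind.
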AnimBlueprint /BRCosmetics/Characters/Player/Male/Medium/Bodies/M_MED_DualParadox/Meshes/M_MED_DualParadox_AnimBP
target_skeleton=&/M_MED_DualParadox_Skeleton
graphs [
 AnimGraph {nodes [
   n1 = "Constraint - dyn_rt_01_lowershirt_01 (RR)"!(ComponentPose, Alpha=1, ConstraintWeights_0=1, ConstraintWeights_1:float=0) -> (Pose) {
    alphaBoolBlend {blendOption=Linear}
    boneToModify {boneName=dyn_rt_01_lowershirt_01}
    constraintSetup [{transformType=Rotation, targetBone {boneName=thigh_r}}, {transformType=Rotation, targetBone {boneName=pelvis}}]
    constraintWeights [1,1]
   }
   "Constraint - dyn_rt_02_lowershirt_01 (RR)"!(ComponentPose=@n1, Alpha=1, ConstraintWeights_0=0.1, ConstraintWeights_1:float=0.9) -> (Pose) {
    alphaBoolBlend {blendOption=Linear}
    boneToModify {boneName=dyn_rt_02_lowershirt_01}
    constraintSetup [{transformType=Rotation, targetBone {boneName=thigh_r}}, {transformType=Rotation, targetBone {boneName=pelvis}}]
    constraintWeights [1,1]
   }
   n2 = "Constraint - dyn_lf_01_lowershirt_01 (RR)"!(ComponentPose, Alpha=1, ConstraintWeights_0=1, ConstraintWeights_1:float=0) -> (Pose) {
    alphaBoolBlend {blendOption=Linear}
    boneToModify {boneName=dyn_lf_01_lowershirt_01}
    constraintSetup [{transformType=Rotation, targetBone {boneName=thigh_l}}, {transformType=Rotation, targetBone {boneName=pelvis}}]
    constraintWeights [1,1]
   }
   "Constraint - dyn_lf_02_lowershirt_01 (RR)"!(ComponentPose=@n2, Alpha=1, ConstraintWeights_0=0.1, ConstraintWeights_1:float=0.9) -> (Pose) {
    alphaBoolBlend {blendOption=Linear}
    boneToModify {boneName=dyn_lf_02_lowershirt_01}
    constraintSetup [{transformType=Rotation, targetBone {boneName=thigh_l}}, {transformType=Rotation, targetBone {boneName=pelvis}}]
    constraintWeights [1,1]
   }
   n3 = "Constraint - dyn_back_lf_01_lowershirt_01 (RR)"!(ComponentPose, Alpha=1, ConstraintWeights_0=0, ConstraintWeights_1:float=1) -> (Pose) {
    alphaBoolBlend {blendOption=Linear}
    boneToModify {boneName=dyn_back_lf_01_lowershirt_01}
    constraintSetup [{transformType=Rotation, targetBone {boneName=thigh_l}}, {transformType=Rotation, targetBone {boneName=pelvis}}]
    constraintWeights [1,1]
   }
   "Constraint - dyn_back_lf_02_lowershirt_01 (RR)"!(ComponentPose=@n3, Alpha=1, ConstraintWeights_0=0, ConstraintWeights_1:float=1) -> (Pose) {
    alphaBoolBlend {blendOption=Linear}
    boneToModify {boneName=dyn_back_lf_02_lowershirt_01}
    constraintSetup [{transformType=Rotation, targetBone {boneName=thigh_l}}, {transformType=Rotation, targetBone {boneName=pelvis}}]
    constraintWeights [1,1]
   }
   n4 = "Constraint - dyn_back_rt_01_lowershirt_01 (RR)"!(ComponentPose, Alpha=1, ConstraintWeights_0=0, ConstraintWeights_1:float=1) -> (Pose) {
    alphaBoolBlend {blendOption=Linear}
    boneToModify {boneName=dyn_back_rt_01_lowershirt_01}
    constraintSetup [{transformType=Rotation, targetBone {boneName=thigh_r}}, {transformType=Rotation, targetBone {boneName=pelvis}}]
    constraintWeights [1,1]
   }
   "Constraint - dyn_back_rt_02_lowershirt_01 (RR)"!(ComponentPose=@n4, Alpha=1, ConstraintWeights_0=0, ConstraintWeights_1:float=1) -> (Pose) {
    alphaBoolBlend {blendOption=Linear}
    boneToModify {boneName=dyn_back_rt_02_lowershirt_01}
    constraintSetup [{transformType=Rotation, targetBone {boneName=thigh_r}}, {transformType=Rotation, targetBone {boneName=pelvis}}]
    constraintWeights [1,1]
   }
   n5 = Control Rig(Source=Copy Pose From Mesh() {bUseAttachedParent=true}) {
    alphaBoolBlend {blendOption=Linear}
    controlRigAssetReference {blueprintRigClass=&/M_MED_DualParadox_DeformRig.M_MED_DualParadox_DeformRig_C}
   }
   n6 = "Constraint - dyn_front_lf_01_lowershirt_MSTR (RR)"!(ComponentPose=Local To Component(LocalPose=@n5), Alpha=1, ConstraintWeights_0=0.7, ConstraintWeights_1:float=0.3) -> (Pose) {
    alphaBoolBlend {blendOption=Linear}
    boneToModify {boneName=dyn_front_lf_01_lowershirt_MSTR}
    constraintSetup [{transformType=Rotation, targetBone {boneName=thigh_l}}, {transformType=Rotation, targetBone {boneName=pelvis}}]
    constraintWeights [1,1]
   }
   n7 = "Constraint - dyn_front_lf_02_lowershirt_MSTR (RR)"!(ComponentPose=@n6, Alpha=1, ConstraintWeights_0=0.9, ConstraintWeights_1:float=0.1) -> (Pose) {
    alphaBoolBlend {blendOption=Linear}
    boneToModify {boneName=dyn_front_lf_02_lowershirt_MSTR}
    constraintSetup [{transformType=Rotation, targetBone {boneName=thigh_l}}, {transformType=Rotation, targetBone {boneName=pelvis}}]
    constraintWeights [1,1]
   }
   n8 = "Constraint - dyn_front_rt_01_lowershirt_MSTR (RR)"!(ComponentPose=@n7, Alpha=1, ConstraintWeights_0=0.7, ConstraintWeights_1:float=0.3) -> (Pose) {
    alphaBoolBlend {blendOption=Linear}
    boneToModify {boneName=dyn_front_rt_01_lowershirt_MSTR}
    constraintSetup [{transformType=Rotation, targetBone {boneName=thigh_r}}, {transformType=Rotation, targetBone {boneName=pelvis}}]
    constraintWeights [1,1]
   }
   n9 = "Constraint - dyn_front_rt_02_lowershirt_MSTR (RR)"!(ComponentPose=@n8, Alpha=1, ConstraintWeights_0=0.9, ConstraintWeights_1:float=0.1) -> (Pose) {
    alphaBoolBlend {blendOption=Linear}
    boneToModify {boneName=dyn_front_rt_02_lowershirt_MSTR}
    constraintSetup [{transformType=Rotation, targetBone {boneName=thigh_r}}, {transformType=Rotation, targetBone {boneName=pelvis}}]
    constraintWeights [1,1]
   }
   n10 = "Transform (Modify) Bone - Bone: spine_02_adj"(ComponentPose=@n9, Alpha=Get Is Crouching Float() -> (Is Crouching Float:float), Translation="0,0,7", Rotation="0,0,0", Scale="1,1,1") {
    rotationMode=BMM_Additive
    rotationSpace=BCS_BoneSpace
    scaleMode=BMM_Replace
    scaleSpace=BCS_BoneSpace
    translationMode=BMM_Additive
    alphaBoolBlend {blendOption=Linear}
    alphaScaleBiasClamp {bInterpResult=true}
    boneToModify {boneName=spine_02_adj}
    translation {z=7}
   }
   n11 = Control Rig!(
    Source=Component To Local(ComponentPose=@n10)
    Alpha
    Baseball_Sliding:double=Get Is Baseball Sliding Float() -> (Is Baseball Sliding Float:float)
   ) -> (Pose) {
    alphaBoolBlend {blendOption=Linear}
    controlRigAssetReference {blueprintRigClass=&/M_MED_DualParadox_LowerVestSkirt_CtrlRig.M_MED_DualParadox_LowerVestSkirt_CtrlRig_C}
   }
   n12 = Get Is Baseball Sliding Float()
   n13 = "Transform (Modify) Bone - Bone: dyn_rt_01_lowershirt_MSTR"(ComponentPose=Local To Component(LocalPose=@n11), Alpha=@n12, Translation="3,6,3", Rotation="120,0,0", Scale="1,1,1") {
    rotationMode=BMM_Additive
    rotationSpace=BCS_BoneSpace
    scaleMode=BMM_Replace
    scaleSpace=BCS_BoneSpace
    translationMode=BMM_Additive
    translationSpace=BCS_BoneSpace
    alphaBoolBlend {blendOption=Linear}
    alphaScaleBiasClamp {interpSpeedIncreasing=50}
    boneToModify {boneName=dyn_rt_01_lowershirt_MSTR}
    rotation {pitch=120}
    translation=<3 6 3>
   }
   n14 = "Transform (Modify) Bone - Bone: dyn_rt_02_lowershirt_MSTR"(ComponentPose=@n13, Alpha=@n12, Translation="0,0,0", Rotation="80,0,0", Scale="1,1,1") {
    rotationMode=BMM_Additive
    rotationSpace=BCS_BoneSpace
    scaleMode=BMM_Replace
    scaleSpace=BCS_BoneSpace
    alphaBoolBlend {blendOption=Linear}
    alphaScaleBiasClamp {interpSpeedIncreasing=50}
    boneToModify {boneName=dyn_rt_02_lowershirt_MSTR}
    rotation {pitch=80}
   }
   n15 = "Transform (Modify) Bone - Bone: dyn_front_rt_01_lowershirt_MSTR"(ComponentPose=@n14, Alpha=@n12, Translation="0,0,3", Rotation="50,0,0", Scale="1,1,1") {
    rotationMode=BMM_Additive
    rotationSpace=BCS_BoneSpace
    scaleMode=BMM_Replace
    scaleSpace=BCS_BoneSpace
    translationMode=BMM_Additive
    translationSpace=BCS_BoneSpace
    alphaBoolBlend {blendOption=Linear}
    alphaScaleBiasClamp {interpSpeedIncreasing=50}
    boneToModify {boneName=dyn_front_rt_01_lowershirt_MSTR}
    rotation {pitch=50}
    translation {z=3}
   }
   n16 = "Transform (Modify) Bone - Bone: dyn_front_rt_02_lowershirt_MSTR"(ComponentPose=@n15, Alpha=@n12, Translation="3,0,12", Rotation="60,-40,-40", Scale="1,1,1") {
    rotationMode=BMM_Additive
    rotationSpace=BCS_BoneSpace
    scaleMode=BMM_Replace
    scaleSpace=BCS_BoneSpace
    translationMode=BMM_Additive
    translationSpace=BCS_BoneSpace
    alphaBoolBlend {blendOption=Linear}
    alphaScaleBiasClamp {interpSpeedIncreasing=50}
    boneToModify {boneName=dyn_front_rt_02_lowershirt_MSTR}
    rotation=<rot 60 -40 -40>
    translation {x=3, z=12}
   }
   n17 = RigidBody!(ComponentPose=@n16, Alpha=Get SkirtAlpha() -> (SkirtAlpha:double), OverrideWorldGravity:Vector=Get GravityOverride(), ExternalForce) -> (Pose) {
    bClampLinearTranslationLimitToRefPose=true
    bOverrideWorldGravity=true
    overridePhysicsAsset=&/M_MED_DualParadox_SkirtPhysics
    simulationSpace=BaseBoneSpace
    alphaBoolBlend {blendOption=Linear}
    baseBoneRef {boneName=pelvis}
    componentAppliedLinearAccClamp=<3000 3000 3000>
    componentLinearAccScale=<0.4 0.3 0.3>
    componentLinearVelScale=<0.4 0.3 0.3>
    simSpaceSettings {
     maxAngularAcceleration=1000
     maxAngularVelocity=3
     maxLinearAcceleration=700
     maxLinearVelocity=500
     velocityScaleZ=0
     worldAlpha=0.75
     externalLinearDragV=<0.2 0.2>
    }
   }
   n18 = Get Is Sky Diving() -> (Is Sky Diving:double)
   n19 = "Transform (Modify) Bone - Bone: dyn_back_lf_02_lowershirt_MSTR"(ComponentPose=@n17, Alpha=@n18, Translation="0,0,0", Rotation="4.5,0,0", Scale="1,1,1") {
    rotationMode=BMM_Additive
    rotationSpace=BCS_BoneSpace
    scaleMode=BMM_Replace
    scaleSpace=BCS_BoneSpace
    alphaBoolBlend {blendOption=Linear}
    alphaScaleBiasClamp {bInterpResult=true, interpSpeedIncreasing=2}
    boneToModify {boneName=dyn_back_lf_02_lowershirt_MSTR}
    rotation {pitch=4.5}
   }
   n20 = "Transform (Modify) Bone - Bone: dyn_back_rt_02_lowershirt_MSTR"(ComponentPose=@n19, Alpha=@n18, Translation="0,0,0", Rotation="4.5,0,0", Scale="1,1,1") {
    rotationMode=BMM_Additive
    rotationSpace=BCS_BoneSpace
    scaleMode=BMM_Replace
    scaleSpace=BCS_BoneSpace
    alphaBoolBlend {blendOption=Linear}
    alphaScaleBiasClamp {bInterpResult=true, interpSpeedIncreasing=2}
    boneToModify {boneName=dyn_back_rt_02_lowershirt_MSTR}
    rotation {pitch=4.5}
   }
   n21 = "Transform (Modify) Bone - Bone: dyn_back_lf_01_lowershirt_MSTR"(ComponentPose=@n20, Alpha=@n18, Translation="0,0,0", Rotation="4.5,0,0", Scale="1,1,1") {
    rotationMode=BMM_Additive
    rotationSpace=BCS_BoneSpace
    scaleMode=BMM_Replace
    scaleSpace=BCS_BoneSpace
    alphaBoolBlend {blendOption=Linear}
    alphaScaleBiasClamp {bInterpResult=true, interpSpeedIncreasing=2}
    boneToModify {boneName=dyn_back_lf_01_lowershirt_MSTR}
    rotation {pitch=4.5}
   }
   n22 = "Transform (Modify) Bone - Bone: dyn_back_rt_01_lowershirt_MSTR"(ComponentPose=@n21, Alpha=@n18, Translation="0,0,0", Rotation="4.5,0,0", Scale="1,1,1") {
    rotationMode=BMM_Additive
    rotationSpace=BCS_BoneSpace
    scaleMode=BMM_Replace
    scaleSpace=BCS_BoneSpace
    alphaBoolBlend {blendOption=Linear}
    alphaScaleBiasClamp {bInterpResult=true, interpSpeedIncreasing=2}
    boneToModify {boneName=dyn_back_rt_01_lowershirt_MSTR}
    rotation {pitch=4.5}
   }
   n23 = Get SpringAlpha() -> (SpringAlpha:double)
   n24 = Get SkyDive Dive Float() -> (SkyDive Dive Float:float)
   n25 = "Transform (Modify) Bone - Bone: deform_upperarm_bk_l"(
    ComponentPose=Spring controller - Bone: deform_knee_out_r!(
     ComponentPose:Component Space Pose Link Structure=Spring controller - Bone: deform_knee_out_l!(ComponentPose:Component Space Pose Link Structure=@n22, Alpha:float=@n23) -> (Pose:Component Space Pose Link Structure) {
      is=title AnimGraphNode_SpringBone
      springDamping=3
      springStiffness=75
      alphaBoolBlend {blendOption=Linear}
      springBone {boneName=deform_knee_out_l}
     }
     Alpha:float=@n23
    ) -> (Pose:Component Space Pose Link Structure) {
     is=title AnimGraphNode_SpringBone
     springDamping=3
     springStiffness=75
     alphaBoolBlend {blendOption=Linear}
     springBone {boneName=deform_knee_out_r}
    }
    Alpha=@n24
    Translation="-10,0,0"
    Rotation="0,0,0"
    Scale="1,1,1"
   ) {
    rotationMode=BMM_Additive
    rotationSpace=BCS_BoneSpace
    scaleMode=BMM_Replace
    scaleSpace=BCS_BoneSpace
    translationMode=BMM_Additive
    translationSpace=BCS_BoneSpace
    alphaBoolBlend {blendOption=Linear}
    alphaScaleBiasClamp {bInterpResult=true}
    boneToModify {boneName=deform_upperarm_bk_l}
    translation {x=-10}
   }
   n26 = "Transform (Modify) Bone - Bone: deform_upperarm_bk_r"(ComponentPose=@n25, Alpha=@n24, Translation="15,0,10", Rotation="0,0,0", Scale="1,1,1") {
    rotationMode=BMM_Additive
    rotationSpace=BCS_BoneSpace
    scaleMode=BMM_Replace
    scaleSpace=BCS_BoneSpace
    translationMode=BMM_Additive
    translationSpace=BCS_BoneSpace
    alphaBoolBlend {blendOption=Linear}
    alphaScaleBiasClamp {bInterpResult=true}
    boneToModify {boneName=deform_upperarm_bk_r}
    rotation=<rot 0 0 0>
    translation {x=15, z=10}
   }
   n27 = Get TacticalSprint Float() -> (TacticalSprint Float:float)
   n28 = "Transform (Modify) Bone - Bone: deform_upperarm_bk_l"(ComponentPose=@n26, Alpha=@n27, Translation="-15,0,15", Rotation="0,0,0", Scale="1,1,1") {
    rotationMode=BMM_Additive
    rotationSpace=BCS_BoneSpace
    scaleMode=BMM_Replace
    scaleSpace=BCS_BoneSpace
    translationMode=BMM_Additive
    translationSpace=BCS_BoneSpace
    alphaBoolBlend {blendOption=Linear}
    alphaScaleBiasClamp {bInterpResult=true}
    boneToModify {boneName=deform_upperarm_bk_l}
    translation {x=-15, z=15}
   }
   n29 = "Transform (Modify) Bone - Bone: deform_upperarm_bk_r"(ComponentPose=@n28, Alpha=@n27, Translation="15,0,15", Rotation="0,0,0", Scale="1,1,1") {
    rotationMode=BMM_Additive
    rotationSpace=BCS_BoneSpace
    scaleMode=BMM_Replace
    scaleSpace=BCS_BoneSpace
    translationMode=BMM_Additive
    translationSpace=BCS_BoneSpace
    alphaBoolBlend {blendOption=Linear}
    alphaScaleBiasClamp {bInterpResult=true}
    boneToModify {boneName=deform_upperarm_bk_r}
    rotation=<rot 0 0 0>
    translation {x=15, z=15}
   }
   n30 = Get IsHangGlider() -> (IsHangGlider:double)
   n31 = "Transform (Modify) Bone - Bone: deform_upperarm_up_base_l"(ComponentPose=@n29, Alpha=@n30, Translation="5,0,15", Rotation="-80,0,0", Scale="1,1,1") {
    rotationMode=BMM_Additive
    rotationSpace=BCS_BoneSpace
    scaleMode=BMM_Replace
    scaleSpace=BCS_BoneSpace
    translationMode=BMM_Additive
    translationSpace=BCS_BoneSpace
    alphaBoolBlend {blendOption=Linear}
    alphaScaleBiasClamp {bInterpResult=true}
    boneToModify {boneName=deform_upperarm_up_base_l}
    rotation {pitch=-80}
    translation {x=5, z=15}
   }
   n32 = "Transform (Modify) Bone - Bone: deform_upperarm_up_base_r"(ComponentPose=@n31, Alpha=@n30, Translation="-5,0,-15", Rotation="-80,0,0", Scale="1,1,1") {
    rotationMode=BMM_Additive
    rotationSpace=BCS_BoneSpace
    scaleMode=BMM_Replace
    scaleSpace=BCS_BoneSpace
    translationMode=BMM_Additive
    translationSpace=BCS_BoneSpace
    alphaBoolBlend {blendOption=Linear}
    alphaScaleBiasClamp {bInterpResult=true}
    boneToModify {boneName=deform_upperarm_up_base_r}
    rotation {pitch=-80}
    translation {x=-5, z=-15}
   }
   n33 = "Transform (Modify) Bone - Bone: dyn_front_lf_02_lowershirt_MSTR"(ComponentPose=@n32, Alpha=Get Is Crouching Float(), Translation="0,0,0", Rotation="5,0,0", Scale="1,1,1") {
    rotationMode=BMM_Additive
    rotationSpace=BCS_BoneSpace
    scaleMode=BMM_Replace
    scaleSpace=BCS_BoneSpace
    translationMode=BMM_Additive
    alphaBoolBlend {blendOption=Linear}
    alphaScaleBiasClamp {bInterpResult=true, interpSpeedDecreasing=2}
    boneToModify {boneName=dyn_front_lf_02_lowershirt_MSTR}
    rotation {pitch=5}
   }
   n34 = Get Is Crouch Melee Float() -> (Is Crouch Melee Float:float)
   n35 = "Transform (Modify) Bone - Bone: dyn_front_lf_01_lowershirt_MSTR"(ComponentPose=@n33, Alpha=@n34, Translation="-5,0,0", Rotation="0,0,0", Scale="1,1,1") {
    rotationMode=BMM_Additive
    rotationSpace=BCS_BoneSpace
    scaleMode=BMM_Replace
    scaleSpace=BCS_BoneSpace
    translationMode=BMM_Additive
    alphaBoolBlend {blendOption=Linear}
    alphaScaleBiasClamp {bInterpResult=true, interpSpeedDecreasing=2}
    boneToModify {boneName=dyn_front_lf_01_lowershirt_MSTR}
    translation {x=-5}
   }
   n36 = "Transform (Modify) Bone - Bone: dyn_front_lf_02_lowershirt_MSTR"(ComponentPose=@n35, Alpha=@n34, Translation="0,0,2.5", Rotation="0,0,0", Scale="1,1,1") {
    rotationMode=BMM_Additive
    rotationSpace=BCS_BoneSpace
    scaleMode=BMM_Replace
    scaleSpace=BCS_BoneSpace
    translationMode=BMM_Additive
    alphaBoolBlend {blendOption=Linear}
    alphaScaleBiasClamp {bInterpResult=true, interpSpeedDecreasing=2}
    boneToModify {boneName=dyn_front_lf_02_lowershirt_MSTR}
    translation {z=2.5}
   }
   Output Pose(Result=Component To Local(ComponentPose=@n36)) {layerGroup=None}
  ]}
 EventGraph {nodes [
   n1 = Event BlueprintUpdateAnimation()
   > n2 = Sequence!(execute) -> (then_0, then_1, then_2:Exec, then_3:Exec, then_4:Exec, then_5:Exec)
   > Is Valid(InputObject=Get Player Pawn() -> (Player Pawn:Fort Player Pawn Athena Object Reference))
   > n3 = Branch(Condition=Get bIsInFrontEnd())
   Set WPO Globals Function(
    execute=@n3.else
    Material=Get Mat Instance Dynamic()
    Overlay Material=Get Overlay Mat Instance Dynamic()
    Skeletal Mesh=Get Skel Mesh()
    "Local Noise Transform  Location"=Get Local Noise Transform Location() -> (Local Noise Transform Location:Vector)
    Local Noise Offset=Get Local Noise Offset()
   )
   > Set WPO Regional Function!(
    execute
    Material:Material Instance Dynamic Object Reference=Get Mat Instance Dynamic()
    Overlay Material:Material Instance Dynamic Object Reference=Get Overlay Mat Instance Dynamic()
    Skeletal Mesh:Skeletal Mesh Component Object Reference=Get Skel Mesh()
    Velocity Transform:Name=foot_r
    Velocity Interp Speed:double=Get Velocity Interp Speed()
    Collision Transform:Name=calf_r
    Collision Transform Axis:EAxis Enum
    Invert Collision Transform Axis:bool=false
    Delta Time X:double=@n1.DeltaTimeX
    Previous Velocity Transform Location:"Vector (by ref)"=Get R1 Velocity Transform Location()
    Previous Velocity Transform Velocity:"Vector (by ref)"=Get R1 Velocity Transform Velocity()
    Region:int=1
   ) -> (then)
   > Set WPO Regional Function(
    Material=Get Mat Instance Dynamic()
    Overlay Material=Get Overlay Mat Instance Dynamic()
    Skeletal Mesh=Get Skel Mesh()
    Velocity Transform=foot_l
    Velocity Interp Speed=Get Velocity Interp Speed()
    Collision Transform=calf_l
    Invert Collision Transform Axis=true
    Delta Time X=@n1.DeltaTimeX
    Previous Velocity Transform Location=Get R2 Velocity Transform Location()
    Previous Velocity Transform Velocity=Get R2 Velocity Transform Velocity()
    Region=2
   )
   > n4 = Branch(Condition=Get bIsSkydiveFloating())
   > n5 = Set WPO Pho Wind Intensity(WPO Pho Wind Intensity=1)
   SetScalarParameterValue(
    execute=[@n6.then, @n5.then]
    self=[Get Mat Instance Dynamic(), Get Overlay Mat Instance Dynamic()]
    ParameterName=ClothSim_WindStrength
    Value=Get WPO Pho Wind Intensity()
   )
   n6 = Set WPO Pho Wind Intensity(execute=@n4.else, WPO Pho Wind Intensity=0)
   n7 = Gravity Override State Function DEPRECATED(
    execute=@n2.then_1
    CharacterPartAnimInstance=Self-Reference()
    GameTimeInSeconds=GetGameTimeInSeconds()
    GravityOverrideParams=Get Gravity Override Params()
   )
   > Set GravityOverride(GravityOverride=@n7)
   Set SkirtAlpha(
    execute=@n2.then_2
    SkirtAlpha:double=float * float!(
     A=Get Is Sky Diving Skirt() -> (Is Sky Diving Skirt:double)
     B=Get Is Sliding() -> (Is Sliding:float)
     C:double=Get Is Swim Sprinting() -> (Is Swim Sprinting:double)
     D:double=Get Is Surface Swimming() -> (Is Surface Swimming:double)
     E:double=Get Is Tactical Sprinting() -> (Is Tactical Sprinting:double)
     F:double=0.85
     G:double=Get Jumping Skirt Alpha() -> (Jumping Skirt Alpha:double)
    ) -> (ReturnValue) {is=title K2Node_PromotableOperator}
   ) -> (Output_Get:double)
   n8 = Sequence(execute=@n2.then_3)
   > n9 = Branch(Condition=Get bIsSkyDiving())
   > Set Is Sky Diving Skirt(Is Sky Diving Skirt:double=0.25) -> (Output_Get:double)
   Set Is Sky Diving Skirt(execute=@n9.else, Is Sky Diving Skirt=1)
   n10 = Branch(execute=@n8.then_1, Condition=Get bIsSliding())
   > Set Is Sliding(Is Sliding:float=0.25) -> (Output_Get:float)
   Set Is Sliding(execute=@n10.else, Is Sliding=1)
   n11 = Branch(execute=@n8.then_2, Condition=Get bIsSwimSprinting() -> (bIsSwimSprinting:bool))
   > Set Is Swim Sprinting(Is Swim Sprinting:double=0.25) -> (Output_Get:double)
   Set Is Swim Sprinting(execute=@n11.else, Is Swim Sprinting=1)
   n12 = Branch(execute=@n8.then_3, Condition=Get bIsSurfaceSwimming())
   > Set Is Surface Swimming(Is Surface Swimming:double=0.25) -> (Output_Get:double)
   Set Is Surface Swimming(execute=@n12.else, Is Surface Swimming=1)
   n13 = Branch(execute=@n8.then_4, Condition=Get bIsTacticalSprinting())
   > Set Is Tactical Sprinting(Is Tactical Sprinting:double=0.5) -> (Output_Get:double)
   Set Is Tactical Sprinting(execute=@n13.else, Is Tactical Sprinting=1)
   n14 = NOT Boolean(A=Get bIsSkyDiving())
   n15 = Get bIsFalling()
   n16 = Branch(execute=@n8.then_5, Condition=OR Boolean(A=Get bIsJumping(), B=AND Boolean(A=@n15, B=@n14)))
   > Set Jumping Skirt Alpha(Jumping Skirt Alpha:double=0.5) -> (Output_Get:double)
   Set Jumping Skirt Alpha(execute=@n16.else, Jumping Skirt Alpha=1)
   n17 = Sequence(execute=@n2.then_4)
   > Set SkyDive Dive Float(SkyDive Dive Float:float="To Float (Boolean)"(InBool=Get bIsSkydiveDiveMode())) -> (Output_Get:float)
   Set TacticalSprint Float(execute=@n17.then_1, TacticalSprint Float:float="To Float (Boolean)"(InBool=Get bIsTacticalSprinting())) -> (Output_Get:float)
   Set Is Crouching Float(execute=@n17.then_2, Is Crouching Float:float="To Float (Boolean)"(InBool=Get bIsCrouching())) -> (Output_Get:float)
   n18 = Get bIsCrouching()
   n19 = Get bIsPlayingMeleeAnim()
   Set Is Crouch Melee Float(execute=@n17.then_3, Is Crouch Melee Float:float="To Float (Boolean)"(InBool=AND Boolean(A=@n19, B=@n18))) -> (Output_Get:float)
   Set Is Baseball Sliding Float(execute=@n17.then_4, Is Baseball Sliding Float:float="To Float (Boolean)"(InBool=Get bIsBaseballSliding())) -> (Output_Get:float)
   n20 = Branch(execute=@n17.then_5, Condition=Get bIsSkyDiving())
   > Set Is Sky Diving(Is Sky Diving:double=1) -> (Output_Get:double)
   Set Is Sky Diving(execute=@n20.else, Is Sky Diving=0)
   n21 = Set pawnVelocitySpeed(execute=@n2.then_5, pawnVelocitySpeed:double=Vector Length(A=Get PawnVelocity())) -> (Output_Get:double)
   > n22 = Set pawnVelocitySpeedCutoff(pawnVelocitySpeedCutoff:double=SelectFloat(A=0, B=1, bPickA="float > float"(A=@n21, B=1200) {is=title K2Node_CallFunction})) -> (Output_Get:double)
   > Set SpringAlpha(SpringAlpha:double=float * float!(A=0.05, B=@n22) -> (ReturnValue)) -> (Output_Get:double)
   Event BlueprintInitializeAnimation()
   > Set Skel Mesh(Skel Mesh=GetOwnerSkeletalMeshComponent())
   n23 = TryGetPawnOwner()
   > Is Valid(InputObject=@n23)
   > n24 = Cast To FortPlayerPawnAthena(Object=@n23)
   > Set Player Pawn(Player Pawn:Fort Player Pawn Athena Object Reference=@n24) -> (Output_Get:Fort Player Pawn Athena Object Reference)
   Event OnCharacterCustomizationLoaded()
   > n25 = Set Skel Mesh(Skel Mesh=GetPartSkeletalMeshComponent())
   > n26 = Get Dynamic Material Instance(SkeletalMeshComponent=@n25)
   > Set Mat Instance Dynamic(Mat Instance Dynamic=@n26)
   > n27 = GetOrCreateDynamicOverlayMaterial(MeshComponent=@n25, MaterialName=None)
   > Set Overlay Mat Instance Dynamic(Overlay Mat Instance Dynamic=@n27)
   Event OnParachuteOpened()
   > Set Is Sky Diving(Is Sky Diving=0)
   > n28 = Switch on EGliderType(Selection=Get PlayerGliderType() -> (PlayerGliderType=HangGlider))
   Set IsHangGlider(execute=[@n28.Cape, @n28.NoGlider, @n28.Surfing, @n28.SurfingSimple, @n28.Umbrella], IsHangGlider:double=0) -> (Output_Get:double)
   Delay(execute=@n28.HangGlider, Duration=1)
   > Set IsHangGlider(IsHangGlider=1)
   Event OnParachuteClosed()
   > Set IsHangGlider(IsHangGlider=0)
  ]}
]
variables [
 body Mesh {type=SkeletalMeshComponent}
 Gravity Override Params {type=GravityOverrideParamsStruct}
 GravityOverride {type=Vector}
 Is Baseball Sliding Float {type=number}
 Is Crouch Melee Float {type=number}
 Is Crouching Float {type=number}
 Is Sky Diving {type=number}
 Is Sky Diving Skirt {type=number}
 Is Sliding {type=number}
 Is Surface Swimming {type=number}
 Is Swim Sprinting {type=number}
 Is Tactical Sprinting {type=number}
 IsHangGlider {type=number}
 Jumping Skirt Alpha {type=number}
 Local Noise Offset {type=Vector}
 Local Noise Transform Location {type=Vector}
 Mat Instance Dynamic {type=MaterialInstanceDynamic}
 Overlay Mat Instance Dynamic {type=MaterialInstanceDynamic}
 pawnVelocitySpeed {type=number}
 pawnVelocitySpeedCutoff {type=number}
 Player Pawn {type=FortPlayerPawnAthena}
 R1 Velocity Transform Location {type=Vector}
 R1 Velocity Transform Velocity {type=Vector}
 R2 Velocity Transform Location {type=Vector}
 R2 Velocity Transform Velocity {type=Vector}
 Skel Mesh {type=SkeletalMeshComponent}
 SkirtAlpha {type=number}
 SkyDive Dive Float {type=number}
 SpringAlpha {type=number}
 TacticalSprint Float {type=number}
 Velocity Interp Speed {type=number, default=30}
 WPO Pho Wind Intensity {type=number}
]
