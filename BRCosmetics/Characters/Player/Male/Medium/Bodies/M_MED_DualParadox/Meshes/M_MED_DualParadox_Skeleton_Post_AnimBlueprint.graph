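# Post-process AnimBlueprint for the M_MED_DualParadox player skeleton that applies the body/head Control Rig to the input pose.
# The rig's alpha becomes 1 only while targeting, crouching and holding a rifle-type weapon, otherwise 0.
# Each update computes that alpha, then refreshes weapon flags from the owning mesh's FortPlayerAnimInstance, so alpha uses the previous update's rifle flag.
# Shotguns, snipers and rocket launchers count as rifles; the pistol flag is unused here.
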AnimBlueprint /BRCosmetics/Characters/Player/Male/Medium/Bodies/M_MED_DualParadox/Meshes/M_MED_DualParadox_Skeleton_Post_AnimBlueprint
target_skeleton=&/M_MED_DualParadox_Skeleton
graphs [
 AnimGraph {nodes [
   n1 = Control Rig(Source=Input Pose(), Alpha=Get Rifle Crouch Target Alpha() -> (Rifle Crouch Target Alpha:float)) {
    alphaBoolBlend {blendOption=Linear}
    controlRigAssetReference {blueprintRigClass=&/M_MED_DualParadox_BodyHead_CtrlRig.M_MED_DualParadox_BodyHead_CtrlRig_C}
   }
   Output Pose(Result=@n1) {layerGroup=None}
  ]}
 EventGraph {nodes [
   Event BlueprintUpdateAnimation()
   > n1 = Sequence()
   n2 = Get bIsTargeting()
   n3 = Get bIsCrouching()
   n4 = Get isRifle() -> (isRifle:bool)
   > Set Rifle Crouch Target Alpha(Rifle Crouch Target Alpha:float="To Float (Boolean)"(InBool=AND Boolean!(A=@n2, B=@n3, C:bool=@n4) -> (ReturnValue))) -> (Output_Get:float)
   n5 = Cast To FortPlayerAnimInstance(execute=@n1.then_1, Object=GetAnimInstance(self=GetOwnerSkeletalMeshComponent()))
   > Set isPistol(isPistol=false)
   > Set isRifle(isRifle:bool=false) -> (Output_Get:bool)
   > n6 = Switch on EFortWeaponCoreAnimation!(execute, Selection:EFortWeaponCoreAnimation Enum=Get WeaponCoreAnim(self=@n5)) -> (
    Melee:Exec
    Pistol:Exec
    Shotgun:Exec
    PaperBlueprint:Exec
    Rifle:Exec
    MeleeOneHand:Exec
    MachinePistol:Exec
    RocketLauncher:Exec
    GrenadeLauncher:Exec
    GoingCommando:Exec
    AssaultRifle:Exec
    TacticalShotgun:Exec
    SniperRifle:Exec
    TrapPlacement:Exec
    ShoulderLauncher:Exec
    AbilityDecoTool:Exec
    Crossbow:Exec
    C4:Exec
    RemoteControl:Exec
    DualWield:Exec
    AR_BullPup:Exec
    AR_ForwardGrip:Exec
    MedPackPaddles:Exec
    SMG_P90:Exec
    AR_DrumGun:Exec
    Consumable_Small:Exec
    Consumable_Large:Exec
    Balloon:Exec
    MountedTurret:Exec
    CreativeTool:Exec
    ExplosiveBow:Exec
    AshtonIndigo:Exec
    AshtonChicago:Exec
    MeleeDualWield:Exec
    Unarmed:Exec
   )
   Set isRifle(execute=[@n6.AssaultRifle, @n6.Rifle, @n6.RocketLauncher, @n6.Shotgun, @n6.SniperRifle, @n6.TacticalShotgun], isRifle=true)
   Set isPistol(execute=@n6.Pistol, isPistol=true)
   TryGetPawnOwner()
  ]}
]
variables [
 isPistol {type=boolean}
 isRifle {type=boolean}
 Rifle Crouch Target Alpha {type=number}
]
