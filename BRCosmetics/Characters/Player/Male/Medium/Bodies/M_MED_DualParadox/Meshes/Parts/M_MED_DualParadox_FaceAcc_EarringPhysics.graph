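# Physics asset for the dangling left earring on the DualParadox male medium character's face accessory.
# Small capsules on three dynamic earring bones form a chain from a master bone, each link constrained with swing limits and a very tight twist limit.
# The master bone is constrained to neck_02 with free swing and twist.
# A large neck_02 capsule gives the neck a physics body alongside the chain.
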PhysicsAsset /BRCosmetics/Characters/Player/Male/Medium/Bodies/M_MED_DualParadox/Meshes/Parts/M_MED_DualParadox_FaceAcc_EarringPhysics
bodies {
 dyn_earring_l_01 [{shapeType=Capsule, length=0.3, radius=0.3, shapeName=dyn_earring_l_01_capsule, center=<0.4349 0 0>, rotation=<rot 90 0 0>}]
 dyn_earring_l_01_MSTR [{shapeType=Capsule, length=1, radius=1, shapeName=dyn_earring_l_01_MSTR_capsule, rotation=<rot 90 0 0>}]
 dyn_earring_l_02 [{shapeType=Capsule, length=0.3, radius=0.3, shapeName=dyn_earring_l_02_capsule, center=<0.4779 0 0>, rotation=<rot 90 0 0>}]
 dyn_earring_l_03 [{shapeType=Capsule, length=0.3, radius=0.3, shapeName=dyn_earring_l_03_capsule, center=<0.4653 0 0>, rotation=<rot 90 0 0>}]
 neck_02 [{shapeType=Capsule, length=5, radius=7.1, shapeName=neck_02_capsule, rotation=<rot 90 0 0>}]
}
constraints [
 {bone1Name=dyn_earring_l_01, bone2Name=dyn_earring_l_01_MSTR, swing1LimitDegrees=45, swing2LimitDegrees=45, twistLimitDegrees=2}
 {bone1Name=dyn_earring_l_02, bone2Name=dyn_earring_l_01, swing1LimitDegrees=45, swing2LimitDegrees=45, twistLimitDegrees=2}
 {bone1Name=dyn_earring_l_03, bone2Name=dyn_earring_l_02, swing1LimitDegrees=45, swing2LimitDegrees=45, twistLimitDegrees=2}
 {
  bone1Name=dyn_earring_l_01_MSTR
  bone2Name=neck_02
  swing1LimitDegrees=45
  swing1Motion=Free
  swing2LimitDegrees=45
  swing2Motion=Free
  twistLimitDegrees=45
  twistMotion=Free
 }
]
