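# Control Rig for the DualParadox male player's hair face-accessory part, applying pose-driven correctives to the right back hair bones.
# A spherical pose reader measures the head bone's orientation relative to spine_05 within an asymmetric active and falloff region.
# Its output weights two chained Modify Transforms nodes that apply fixed Z-axis rotations to the two right back hair master bones, the second rotating further.
# Debug drawing is enabled on the pose reader.
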ControlRigBlueprint /BRCosmetics/Characters/Player/Male/Medium/Bodies/M_MED_DualParadox/Meshes/Parts/M_MED_DualParadox_FaceAcc_Hair_CtrlRig
graphs [
 RigVMFunctionLibrary {links [], nodes []}
 RigVMModel {nodes [
   RigUnit_BeginExecution = &RigUnit_BeginExecution()
   n1 = &RigUnit_SphericalPoseReader(
    ActiveRegionScaleFactors=(PositiveWidth=0.25,NegativeWidth=2,PositiveHeight=1,NegativeHeight=1)
    ActiveRegionSize=0.1
    Debug=(bDrawDebug=true,DebugScale=25,DebugSegments=20,DebugThickness=0.25)
    DriverAxis=(X=0,Y=-1,Z=0)
    DriverItem=(Type=Bone,Name="head")
    FalloffRegionScaleFactors=(PositiveWidth=1.2,NegativeWidth=1,PositiveHeight=2,NegativeHeight=2)
    FalloffSize=0.2
    OptionalParentItem=(Type=Bone,Name="spine_05")
    RotationOffset=(X=-80,Y=0,Z=-43)
    ExecutePin=@RigUnit_BeginExecution.ExecutePin
   )
   > &RigUnit_ModifyTransforms(
    ItemToModify=((Item=(Type=Bone,Name="dyn_bk_hair_r_01_MSTR"),Transform=(Rotation=(X=0,Y=0,Z=0,W=1),Translation=(X=0,Y=0,Z=0),Scale3D=(X=1,Y=1,Z=1))))
    ItemToModify.0.Transform.Rotation=&RigVMFunction_MathQuaternionFromEuler(Euler=(X=0,Y=0,Z=25), RotationOrder=XYZ).Result
    Weight=@n1.OutputParam
   )
   > &RigUnit_ModifyTransforms(
    ItemToModify=((Item=(Type=Bone,Name="dyn_bk_hair_r_02_MSTR"),Transform=(Rotation=(X=0,Y=0,Z=0,W=1),Translation=(X=0,Y=0,Z=0),Scale3D=(X=1,Y=1,Z=1))))
    ItemToModify.0.Transform.Rotation=&RigVMFunction_MathQuaternionFromEuler(Euler=(X=0,Y=0,Z=45), RotationOrder=XYZ).Result
    Weight=@n1.OutputParam
   )
  ]}
]
hierarchy [
 root {loc=<0 0 0>, rot=<0 0 0 1>}
 pelvis {parent=root, loc=<0 -0.7071 93.1191>, rot=<0.0127 0.707 -0.0127 -0.707>}
 spine_01 {parent=pelvis, loc=<6.7494 -0.6388 0>, rot=<0 0 0.034 -0.9994>}
 spine_02 {parent=spine_01, loc=<7.2518 -0.035 0>, rot=<0 0 0.0238 -0.9997>}
 spine_03 {parent=spine_02, loc=<6.0104 0 0>, rot=<0 0 0.0149 -0.9999>}
 spine_04 {parent=spine_03, loc=<8.1024 0 0>, rot=<0 0 -0.0132 -0.9999>}
 spine_05 {parent=spine_04, loc=<10.0146 0 0>, rot=<0 0 -0.0015 -1>}
 neck_01 {parent=spine_05, loc=<11.3138 0.1529 0>, rot=<0 0 -0.2003 -0.9797>}
 neck_02 {parent=neck_01, loc=<4.5627 0 0>, rot=<0 0 0.0504 -0.9987>}
 head {parent=neck_02, loc=<5.3914 0 0>, rot=<0 0 0.115 -0.9934>}
 dyn_hair_tips_01 {parent=head, loc=<13.4518 12.0313 -5.3958>, rot=<0.0033 -0.8049 -0.5934 0.0025>}
 dyn_hair_tips_02 {parent=head, loc=<17.3625 10.5793 -4.696>, rot=<-0.4172 -0.5709 -0.4219 -0.5674>}
 dyn_hair_tips_03 {parent=head, loc=<14.5527 12.8843 1.9201>, rot=<-0.0027 0.6435 0.7655 -0.0032>}
 dyn_top_hair_01_MSTR {parent=head, loc=<24.37 4.9932 0.7677>, rot=<0.4934 0.2771 -0.1504 -0.8106>}
 dyn_top_hair_01 {parent=dyn_top_hair_01_MSTR, loc=<0 0 0>, rot=<0 0 0 1>}
 dyn_top_hair_02_MSTR {parent=head, loc=<20.774 10.4841 1.9517>, rot=<0.4154 0.4668 -0.4372 -0.6468>}
 dyn_top_hair_02 {parent=dyn_top_hair_02_MSTR, loc=<0 0 0>, rot=<0 0 0 1>}
 dyn_top_hair_03_MSTR {parent=head, loc=<21.6833 2.8451 7.4974>, rot=<0.2985 0.6241 -0.0342 -0.7212>}
 dyn_top_hair_03 {parent=dyn_top_hair_03_MSTR, loc=<0 0 0>, rot=<0 0 0 1>}
 dyn_top_hair_04_MSTR {parent=head, loc=<16.458 9.4149 7.2791>, rot=<0.3053 0.6725 -0.3883 -0.551>}
 dyn_top_hair_04 {parent=dyn_top_hair_04_MSTR, loc=<0 0 0>, rot=<0 0 0 1>}
 dyn_top_hair_05_MSTR {parent=head, loc=<13.3843 1.7433 11.325>, rot=<0.2919 0.7436 0.056 -0.5989>}
 dyn_top_hair_05 {parent=dyn_top_hair_05_MSTR, loc=<0 0 0>, rot=<0 0 0 1>}
 dyn_bk_hair_l_01_MSTR {parent=head, loc=<11.433 -4.384 -9.2206>, rot=<-0.4184 -0.4767 0.5117 -0.5795>}
 dyn_bk_hair_l_01 {parent=dyn_bk_hair_l_01_MSTR, loc=<0 0 0>, rot=<0 0 0 1>}
 dyn_bk_hair_r_01_MSTR {parent=head, loc=<11.3121 -10.0925 4.1836>, rot=<0.2331 0.2739 0.6556 -0.664>}
 dyn_bk_hair_r_01 {parent=dyn_bk_hair_r_01_MSTR, loc=<0 0 0>, rot=<0 0 0 1>}
 dyn_bk_hair_r_02_MSTR {parent=head, loc=<8.3902 -5.4891 8.6623>, rot=<0.5186 0.5102 0.4852 -0.4851>}
 dyn_bk_hair_r_02 {parent=dyn_bk_hair_r_02_MSTR, loc=<0 0 0>, rot=<0 0 0 1>}
 dyn_earring_l_01_MSTR {parent=head, loc=<1.7896 1.091 -8.5601>, rot=<0.0041 -0.9926 0.1212 -0.0005>}
 dyn_earring_l_01 {parent=dyn_earring_l_01_MSTR, loc=<0 0 0>, rot=<0 0 0 1>}
 dyn_earring_l_02 {parent=dyn_earring_l_01, loc=<1.009 0 0>, rot=<0 0 0 1>}
 dyn_earring_l_03 {parent=dyn_earring_l_02, loc=<0.9205 0 0>, rot=<0 0 0 1>}
]
